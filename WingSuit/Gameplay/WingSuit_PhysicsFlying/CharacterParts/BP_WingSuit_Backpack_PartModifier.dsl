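; Character part modifier for the wingsuit backpack that keeps its mesh and idle VFX in sync with the player's wingsuit state.
; After resurrection-in, it starts four gameplay tag wait tasks on the associated pawn.
; The unfurl tag hides the part's skeletal mesh and children, and removal of the wingsuit flying tag shows them again.
; The granted cooldown tag deactivates the idle VFX when added and reactivates it when removed.
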(event EventOnPostResIn
  (bind _returnvalue (CustomCharacter|GetAssociatedPlayerPawn))
  (bind _returnvalue_2 (CallFunction|GetIdleVFXComponent))
  (|CallParentFunction)
  (Ability|Tasks|WaitGameplayTagAddToActor _returnvalue "(TagName=\"WingSuit.Unfurl\")"
    (:then
      (Ability|Tasks|WaitGameplayTagRemoveFromActor _returnvalue "(TagName=\"MovementMode.Flying.WingSuit\")"
        (:then
          (Ability|Tasks|WaitGameplayTagAddToActor (CustomCharacter|GetAssociatedPlayerPawn) "(TagName=\"granted.wingsuit.cooldown\")"
            (:then
              (Ability|Tasks|WaitGameplayTagRemoveFromActor (CustomCharacter|GetAssociatedPlayerPawn) "(TagName=\"granted.wingsuit.cooldown\")"
                (:Removed
                  (Components|Activation|Activate _returnvalue_2 false))))
            (:Added
              (Components|Activation|Deactivate _returnvalue_2))))
        (:Removed
          (Rendering|SetVisibility (CustomCharacter|GetSkeletalMeshForPartOnPlayerPawn) true true))))
    (:Added
      (Rendering|SetVisibility (CustomCharacter|GetSkeletalMeshForPartOnPlayerPawn) false true))))

(fn ConstructionScript
  (|CallParentFunction))
