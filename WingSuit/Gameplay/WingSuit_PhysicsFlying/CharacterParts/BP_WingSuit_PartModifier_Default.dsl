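; Default character part modifier for the physics-flying wingsuit that hands the part's skeletal mesh to the wingsuit weapon.
; On post-res-in, after the parent logic, it casts the associated player pawn's current weapon to B_Weap_WingSuit_PhysicsFlying.
; It then stores the pawn's skeletal mesh component for this part in that weapon's CharacterPart variable.
; The construction script only calls the parent.
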(event EventOnPostResIn
  (|CallParentFunction)
  (bind _asb_weap_wing_suit_physics_flying (Utilities|Casting|CastToB_Weap_WingSuit_PhysicsFlying (Weapon|GetCurrentWeapon (CustomCharacter|GetAssociatedPlayerPawn))))
  (bind _asskeletal_mesh_component (Utilities|Casting|CastToSkeletalMeshComponent (CustomCharacter|GetSkeletalMeshForPartOnPlayerPawn)))
  (Variables|/WingSuit/Gameplay/WingSuit_PhysicsFlying/B_Weap_WingSuit_PhysicsFlying.B_Weap_WingSuit_PhysicsFlying_C|SetCharacterPart _asskeletal_mesh_component _asb_weap_wing_suit_physics_flying))

(fn ConstructionScript
  (|CallParentFunction))
